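; Bug Biome loot chest actor that players break open, notifying listeners on approach, damage and opening.
; Looting destroys it; death spawns a lightweight destroyed actor, using a no-RVT class in tunnels, unless suppressed by UndoSpawn or detachment death, which plays burst VFX.
; Damage plays impact audio plus a hit-notify sound for the local instigator.
; Construction sets blue glow materials, rare-chest Niagara effects and a spawn-chance preview from an enum and multiplier.
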(event EventBeginPlay
  (|CallParentFunction))

(event Collision|EventActorBeginOverlap
  (params (OtherActor :type Actor%20Object%20Reference)))

(event EventTick
  (params (DeltaSeconds :type Float%20%28single-precision%29)))

(event OnComponentBeginOverlap(PlayerDetector)
  (params (OverlappedComponent :type Primitive%20Component%20Object%20Reference) (OtherActor :type Actor%20Object%20Reference) (OtherComp :type Primitive%20Component%20Object%20Reference) (OtherBodyIndex :type Integer) (bFromSweep :type Boolean) (SweepResult :type Hit%20Result%20Structure%20%28by%20ref%29))
  (Networking|SwitchHasAuthority
    (:Authority
      (CallFunction|PlayerApproachedMulticast))))

(event Custom|PlayerApproached_Multicast
  (Default|CallOnPlayerApproached))

(event EventOnLoot
  (Default|CallOnOpened)
  (Default|CallOnAttackedorOpened)
  (Actor|DestroyActor))

(event EventOnDamageServer
  (params (Damage :type Float%20%28single-precision%29) (DamageTags :type Gameplay%20Tag%20Container%20Structure%20%28by%20ref%29) (Momentum :type Vector :default "0, 0, 0") (HitInfo :type Hit%20Result%20Structure%20%28by%20ref%29) (InstigatedBy :type Controller%20Object%20Reference) (DamageCauser :type Actor%20Object%20Reference) (EffectContext :type Gameplay%20Effect%20Context%20Handle%20Structure))
  (CallFunction|OnDamageMulticast))

(event Custom|OnDamage%20Multicast
  (Default|CallOnAttackedorOpened))

(event EventOnDeathServer
  (params (Damage :type Float%20%28single-precision%29) (DamageTags :type Gameplay%20Tag%20Container%20Structure%20%28by%20ref%29) (Momentum :type Vector :default "0, 0, 0") (HitInfo :type Hit%20Result%20Structure%20%28by%20ref%29) (InstigatedBy :type Controller%20Object%20Reference) (DamageCauser :type Actor%20Object%20Reference) (EffectContext :type Gameplay%20Effect%20Context%20Handle%20Structure))
  (Utilities|FlowControl|Sequence
    (:then_0
      (if (GameplayTags|HasTag DamageTags "(TagName=\"Gameplay.Effect.InstantDeath.Detachment\")" true)
        (Variables|Default|SetDontSpawnDestroyedVersion true)
        (CallFunction|PlayDeathVFXOnDettachment)))
    (:then_1
      ; Spawn lightweight destroyed actor if needed
      (if (not (Variables|Default|GetDontSpawnDestroyedVersion))
        (bind _returnvalue (Game|SpawnActorfromClass (select (Variables|Default|GetbIsInTunnel) (Variables|Default|GetDestroyedVersionClassNoRVT) (Variables|Default|GetDestroyedVersionClass)) (Transformation|GetActorTransform) "Undefined" "MultiplyWithRoot" 0 (Variables|Actor|GetInstigator)))
        (Default|CallOnDestroyedMeshSpawned _returnvalue)))))

(event Custom|PlayDeathVFXOnDettachment
  (Niagara|SpawnSystematLocation "/BugBiomeGameplay/Assets/VFX/NS_LootSac_Open.NS_LootSac_Open" (Transformation|GetWorldLocation (Variables|Default|GetTreasureLight))))

(event Custom|UndoSpawn
  (Variables|Default|SetDontSpawnDestroyedVersion true)
  (Gameplay|SilentDie false))

(event Custom|DestroyAttachedActors
  (for _array_element (Actor|GetAttachedActors true true)
    (Actor|DestroyActor _array_element)
    (Development|PrintString (Utilities|GetDisplayName _array_element) true true "(R=0.000000,G=0.660000,B=1.000000,A=1.000000)" 3e+01)))

(event EventOnDamagePlayEffects
  (params (Damage :type Float%20%28single-precision%29) (DamageTags :type Gameplay%20Tag%20Container%20Structure%20%28by%20ref%29) (Momentum :type Vector :default "0, 0, 0") (HitInfo :type Hit%20Result%20Structure%20%28by%20ref%29) (InstigatedBy :type Fort%20Pawn%20Object%20Reference) (DamageCauser :type Actor%20Object%20Reference) (EffectContext :type Gameplay%20Effect%20Context%20Handle%20Structure))
  (bind (_bblockinghit _binitialoverlap _time _distance _location) (Collision|BreakHitResult HitInfo))
  ; On damage, play hitmarker sound for instigator, impact sound for 3p
  (Audio|PlaySoundatLocation "/BugBiomeGameplay/Gameplay/Rewards/Audio/MSS_LootSac_Impact.MSS_LootSac_Impact" _location)
  (if (Pawn|IsLocallyViewed InstigatedBy)
    (Audio|PlaySound2D "/BugBiomeGameplay/Gameplay/Rewards/Audio/MSS_LootSac_HitNotify.MSS_LootSac_HitNotify")))

(fn BlueprintGetInteractionString
  (params (InteractingPawn :type Fort%20Pawn%20Object%20Reference) (InteractionBeingAttempted :type EInteractionBeingAttempted%20Enum :default "FirstInteraction"))
  (return "NSLOCTEXT(\"[6791EE5206C94ADE4DE823F8C1F2BF8B]\", \"241CD5D24AC9D23712C018903D8EB693\", \"Break Open\")"))

(fn CalculateSpawnChance
  (Utilities|FlowControl|Sequence
    (:then_0
      (switch Utilities|FlowControl|Switch|SwitchonLootSackSpawnChance (Variables|Default|GetSpawnChance)
        (:NewEnumerator2
          (LocalVariables|SetSpawn%20Chance 1e+02))
        (:NewEnumerator0
          (LocalVariables|SetSpawn%20Chance (* (Variables|Default|GetSpawnChanceMultiplier) 7e+01)))
        (:NewEnumerator4
          (LocalVariables|SetSpawn%20Chance (* (Variables|Default|GetSpawnChanceMultiplier) 0)))))
    (:then_1
      (return (Math|Float|Clamp(Float) (LocalVariables|GetSpawn%20Chance) 0.0 1e+02)))))

(fn ConstructionScript
  (bind _fx_glow___smoke_sheet (Variables|Default|GetFX%20Glow%20&%20Smoke%20Sheet))
  (|CallParentFunction)
  (Rendering|Material|SetDefaultCustomPrimitiveDataFloat _fx_glow___smoke_sheet 0 1.0) ; WPO_HideMesh
  (Rendering|Material|SetScalarParameterValueonMaterials _fx_glow___smoke_sheet "Corona_Opacity" 2e+01)
  (Rendering|Material|SetScalarParameterValueonMaterials _fx_glow___smoke_sheet "Glow_Opacity" 7.0)
  (Rendering|Material|SetVectorParameterValueonMaterials _fx_glow___smoke_sheet "Corona_Colour" "0.450000,0.874998,3.000000")
  (Rendering|Material|SetVectorParameterValueonMaterials _fx_glow___smoke_sheet "Glow_Colour" "0.050000,0.240000,1.000000")
  (Rendering|Material|SetScalarParameterValueonMaterials _fx_glow___smoke_sheet "UseBaseColor" 0.0)
  (Rendering|Material|SetColorParameterValueonMaterials _fx_glow___smoke_sheet "SecColor" "(R=0.000000,G=0.268970,B=1.000000,A=1.000000)")
  (Rendering|Material|SetColorParameterValueonMaterials _fx_glow___smoke_sheet "SpriteColor_Low" "(R=0.000000,G=0.369341,B=1.000000,A=1.000000)")
  (Variables|Default|SetCustom%20Niagara%20Effects (Utilities|Array|MakeArray (Variables|Default|GetRare%20Chest%20Visuals)))
  (bind _spawn_chance1 (CallFunction|CalculateSpawnChance))
  (Variables|Default|SetActual%20Spawn%20Chance%20%28Preview%29 _spawn_chance1))
