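# One-shot stereo MetaSound for a loot sac hit notification, playing one of eight SwarmSpawner hit marker waves on the 2D instigator hit notification sound class.
# On play, a character affiliation check gates a random wave pick, which rolls a pitch shift and starts the wave player.
# An azimuth panner places the output, scaled by random volume; pitch and volume ranges default to no variation.
# A concurrency override allows only one instance at a time.
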MetaSoundSource /BugBiomeGameplay/Gameplay/Rewards/Audio/MSS_LootSac_HitNotify
inputs [
 Character.Affiliation {type=Enum:FortAudioAffiliation}
 Character.Health {type=Float}
 Character.UniqueID {type=Int32}
 Max Pitch {type=Float}
 Max Random Volume {type=Float}
 Min Pitch {type=Float}
 Min Random Volume {type=Float}
 Random.Pitch.Base {type=Float}
 Random.Pitch.Spread {type=Float}
 Random.Volume.Random Volume Base {type=Float}
 Random.Volume.Spread {type=Float}
 Sounds {type=WaveAsset:Array}
 StereoSpread {type=Float}
 UE.Source.OnPlay {type=Trigger}
 UE.Source.Orientation.Azimuth {type=Float}
]
interfaces [Fort.Character:1.0, UE.OutputFormat.Stereo:1.0, UE.Source.OneShot:1.0, UE.Source:1.1]
outputs [
 "UE.OutputFormat.Stereo.Audio:0" {type=Audio}
 "UE.OutputFormat.Stereo.Audio:1" {type=Audio}
 UE.Source.OneShot.OnFinished {type=Trigger}
]
pages [Default {
  input_defaults {
   Character.Affiliation=0
   Character.Health=0
   Character.UniqueID=0
   Max Pitch=1
   Max Random Volume=1
   Min Pitch=1
   Min Random Volume=1
   Random.Pitch.Base=0
   Random.Pitch.Spread=0
   Random.Volume.Random Volume Base=0
   Random.Volume.Spread=0
   StereoSpread=0
   UE.Source.OnPlay=null
   UE.Source.Orientation.Azimuth=0
   Sounds [
    &/SwarmSpawner_HitMarker_08
    &/SwarmSpawner_HitMarker_07
    &/SwarmSpawner_HitMarker_06
    &/SwarmSpawner_HitMarker_05
    &/SwarmSpawner_HitMarker_04
    &/SwarmSpawner_HitMarker_03
    &/SwarmSpawner_HitMarker_02
    &/SwarmSpawner_HitMarker_01
   ]
  }
  nodes [
   ArrayRandomGet_0 = Array.Random Get:WaveAsset:Array(In Array=^Sounds, Next=&TriggerCompareAffiliation(A=^Character.Affiliation, Compare=^UE.Source.OnPlay).True) {also {defaults {Weights []}}}
   n1 = UE.RandomFloat(Max=^Max Pitch, Min=^Min Pitch, Next=@ArrayRandomGet_0.On Next)
   WavePlayer_0 = UE.Wave Player:Stereo(Play=@n1.On Next, Pitch Shift=@n1.Value, Wave Asset=UE.Reroute(Value=UE.Reroute(Value=@ArrayRandomGet_0.Value).Value).Value) {also {version [1,1]}}
   n2 = &AzimuthPannerStereo(StereoSpread=^StereoSpread, Azimuth=^UE.Source.Orientation.Azimuth, In L=@WavePlayer_0.Out Left, In R=@WavePlayer_0.Out Right)
   n3 = UE.RandomFloat(Max=^Max Random Volume, Min=^Min Random Volume, Next=^UE.Source.OnPlay)
   Multiply_0 = UE.Multiply:Audio by Float(PrimaryOperand=@n2.Out L, AdditionalOperands=@n3.Value)
   Multiply_1 = UE.Multiply:Audio by Float(PrimaryOperand=@n2.Out R, AdditionalOperands=@n3.Value)
  ]
  links [
   {dst="@UE.OutputFormat.Stereo.Audio:0", src=Multiply_0.Out}
   {dst="@UE.OutputFormat.Stereo.Audio:1", src=Multiply_1.Out}
   {dst="@UE.Source.OneShot.OnFinished", src=WavePlayer_0.On Finished}
  ]
 }]
properties {
 assetImportData=&self:AssetImportData
 bOverrideConcurrency=true
 numChannels=2
 outputFormat=Stereo
 sampleRate=48000
 soundClassObject=&Fort_2DHitNotifications_Instigate_PSM
 volume=0.2
 concurrencyOverrides {maxCount=1}
}
