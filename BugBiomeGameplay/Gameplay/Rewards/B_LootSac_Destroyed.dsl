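; Blueprint B_LootSac_Destroyed, by its name the destroyed-state actor of a loot sac reward in the bug biome.
; On BeginPlay it flushes net dormancy on itself, pushing its current state to clients despite dormancy.
; The construction script is empty and no other logic is shown.
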(event EventBeginPlay
  (Networking|FlushNetDormancy))

(fn ConstructionScript)
